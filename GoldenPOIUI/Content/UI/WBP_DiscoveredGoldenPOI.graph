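# Discoverability display extension widget that shows a gold-textured background when the player arrives at a tagged golden POI.
# On the common arrival event, it looks up the location tag's context tags through the game state's minimap data override component.
# A non-exact match on Athena.POIText.Sideways, apparently reused as the golden marker, makes the widget hit-test invisible (shown); otherwise it collapses.
# An invalid component leaves visibility unchanged.
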WidgetBlueprint /GoldenPOIUI/Content/UI/WBP_DiscoveredGoldenPOI
graphs [EventGraph {nodes [
   n1 = Event OnCommonArrivalEvent!() -> (OutputDelegate, then, NewTag:"Gameplay Tag Structure (by ref)") {is=title K2Node_Event}
   n2 = GetComponentByClass!(
    self:Actor Object Reference=GetGameStateAthena!() -> (ReturnValue:Fort Game State Athena Object Reference)
    ComponentClass:Actor Component Class Reference=/Script/FortniteGame.FortGameStateComponent_MiniMapDataOverride
   ) -> (ReturnValue:Fort Game State Component Mini Map Data Override Object Reference)
   > Is Valid(InputObject=@n2)
   > n3 = Branch(Condition=HasTag!(
     TagContainer:"Gameplay Tag Container Structure (by ref)"=GetContextTagsForLocationTag!(self:Fort Game State Component Mini Map Data Override Object Reference=@n2, LocationTag:Gameplay Tag Structure=@n1.NewTag) -> (ReturnValue:Gameplay Tag Container Structure)
     Tag:Gameplay Tag Structure=(TagName="Athena.POIText.Sideways")
     bExactMatch:bool=false
    ) -> (ReturnValue:bool))
   > SetVisibility(InVisibility=HitTestInvisible)
   SetVisibility(execute=@n3.else, InVisibility=Collapsed)
   Event PreConstruct()
   Event Construct()
   Event Tick()
  ]}]
variables []
widgets {
 info {
  inheritedWidgetCount=1
  parentClass=/Script/FortniteUI.AthenaDiscoverabilityDisplay_Extension
  rootWidgetClass=&Overlay
  widgetCount=4
 }
 widgets [
  Overlay_33 {widgetClassPath=&Overlay, widget=Overlay_33}
  SizeBox_0 {
   widgetClassPath=&SizeBox
   parent=Overlay_33
   widget=SizeBox_0
   properties {heightOverride=200, minDesiredHeight=100, minDesiredWidth=300, widthOverride=600}
   slot {class=OverlaySlot, properties {horizontalAlignment=HAlign_Fill, verticalAlignment=VAlign_Fill}}
  }
  Image_GoldenBackground {
   widgetClassPath=&Image
   bIsVariable=true
   parent=SizeBox_0
   widget=Image_GoldenBackground
   properties {brush {
     bIsDynamicallyLoaded=false
     drawAs=Image
     imageType=NoImage
     mirroring=NoMirror
     resourceName=None
     resourceObject=/Game/Environments/Apollo/Sets/Yacht/Textures/Gold_D
     tiling=NoTile
     imageSize=<256 64>
     margin=<ltrb 0 0 0 0>
     outlineSettings {
      bUseBrushTransparency=false
      roundingType=HalfHeightRadius
      width=0
      color {colorUseRule=UseColor_Specified, specifiedColor=<rgba 0 0 0 0>}
      cornerRadii=<0 0 0 0>
     }
     tintColor {colorUseRule=UseColor_Specified, specifiedColor=<rgba 1 1 1 1>}
     uVRegion {bIsValid=false, max=<0 0>, min=<0 0>}
    }}
   slot {class=SizeBoxSlot}
  }
  TutorialHighlight {widgetClassPath=&TutorialHighlight}
 ]
}
