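# Translucent UI material for the Athena HUD that builds an animated, gold-orange swirl gradient, named for special POIs.
# Two panning samples of a swirl flow texture, with different tiling and speed parameters, blend with Light, Medium and Dark color parameters.
# Nested constant-alpha lerps, some extrapolating beyond one, shape the contrast.
# A parameterized linear gradient function, with offset and rotation, is mixed in at low weight.
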Material /Game/Athena/HUD/Materials/M_UI_LinearGradient_SpecialPOI
graph {
 properties {blendMode=BLEND_Translucent, materialDomain=MD_UI, pixelDepthOffsetMode=PDOM_Legacy, usedShadingModels=Default Lit}
 nodes [
  n1 = VectorParameter() {
   group=Gradient Settings
   parameterName=GradientRotationCenter
   sortPriority=5
   channelNames {a="", b="", g=Center Y, r=Center X}
   defaultValue=<rgba 0.5 0.5 0 0>
  }
  n2 = MaterialFunctionCall(
   GradientStart=ScalarParameter() {group=Gradient Settings, parameterName=GradientStart, sortPriority=1}
   GradientEnd=ScalarParameter() {defaultValue=1, group=Gradient Settings, parameterName=GradientEnd, sortPriority=2}
   RotationAngle=ScalarParameter() {defaultValue=0.5, group=Gradient Settings, parameterName=GradientRotation, sortPriority=4}
   RotationCenter=@n1
   GradientOffset=ScalarParameter() {group=Gradient Settings, parameterName=GradientOffset, sortPriority=3}
  ) {materialFunction=&MF_UI_LinearGradient}
  LinearInterpolate(Alpha=@n2.Result)
  n3 = TextureCoordinate() {uTiling=-1, vTiling=-0.5}
  n4 = ScalarParameter() {defaultValue=0.07, parameterName=PanSpeed_Layer2}
  n5 = TextureSample(UVs=Panner(Coordinate=TextureCoordinate() {uTiling=0.5}, Speed=ScalarParameter() {defaultValue=0.08, parameterName=PanSpeed_Layer1}) {speedX=-0.06, speedY=0.025}) {texture=/Game/Global/Textures/Default/Flowmaps/T_SwirlFlow_02_D}
  n6 = LinearInterpolate(
   A=TextureSample(UVs=Panner(Coordinate=@n3, Speed=@n4) {speedX=-0.06, speedY=0.025}) {texture=&T_SwirlFlow_02_D}.RGB
   B=@n5.RGB
  ) {constAlpha=3}
  n7 = LinearInterpolate(
   A=VectorParameter() {parameterName=Light, defaultValue=<rgba 0.9576 0.6517 0.1795 1>}.RGB
   B=VectorParameter() {parameterName=Medium, defaultValue=<rgba 0.9375 0.5744 0 1>}.RGB
  )
  n8 = Multiply(A=@n5.RGB) {constB=3}
  n9 = LinearInterpolate(A=VectorParameter() {parameterName=Dark, defaultValue=<rgba 1 0.4587 0.0603 1>}.RGB, B=@n7) {constAlpha=0.45}
  LinearInterpolate(
   A=LinearInterpolate(A=LinearInterpolate(A=@n6, B=@n7), B=LinearInterpolate(A=@n8, B=@n9) {constAlpha=1.4}) {constAlpha=0.6, constB=1.5}
   B=LinearInterpolate(Alpha=@n2.Result)
  ) {constAlpha=0.1}
 ]
}
