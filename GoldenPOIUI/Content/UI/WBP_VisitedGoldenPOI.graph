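# WBP_VisitedGoldenPOI is a discoverability display extension widget that adds a special-POI gradient background when the player arrives at certain named locations.
# On the common arrival event, it fetches the MiniMapDataOverride component from the Athena game state and requests context tags for the arrived location tag.
# The widget shows itself without self hit-testing if those tags include Athena.POIText.Sideways, and collapses otherwise or when the component is invalid.
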WidgetBlueprint /GoldenPOIUI/Content/UI/WBP_VisitedGoldenPOI
graphs [EventGraph {nodes [
   n1 = Event OnCommonArrivalEvent!() -> (OutputDelegate, then, NewTag:"Gameplay Tag Structure (by ref)") {is=title K2Node_Event}
   n2 = GetComponentByClass!(
    self:Actor Object Reference=GetGameStateAthena!() -> (ReturnValue:Fort Game State Athena Object Reference)
    ComponentClass:Actor Component Class Reference=/Script/FortniteGame.FortGameStateComponent_MiniMapDataOverride
   ) -> (ReturnValue:Fort Game State Component Mini Map Data Override Object Reference)
   > n3 = Is Valid(InputObject=@n2)
   SetVisibility(execute=@n3.Is Not Valid, InVisibility=Collapsed)
   n4 = GetContextTagsForLocationTag!(self:Fort Game State Component Mini Map Data Override Object Reference=@n2, LocationTag:Gameplay Tag Structure=@n1.NewTag) -> (ReturnValue:Gameplay Tag Container Structure)
   SetVisibility(
    execute=@n3.Is Valid
    InVisibility=Select(
     Option 0=Collapsed
     Option 1=SelfHitTestInvisible
     Index=HasTag!(
      TagContainer:"Gameplay Tag Container Structure (by ref)"=@n4
      Tag:Gameplay Tag Structure=(TagName="Athena.POIText.Sideways")
      bExactMatch:bool=false
     ) -> (ReturnValue:bool)
    )
   )
  ]}]
variables []
widgets {
 info {
  inheritedWidgetCount=1
  parentClass=/Script/FortniteUI.AthenaDiscoverabilityDisplay_Extension
  rootWidgetClass=&Overlay
  widgetCount=3
 }
 widgets [
  Overlay_33 {widgetClassPath=&Overlay, widget=Overlay_33}
  Image_GoldenBackground {
   widgetClassPath=&Image
   bIsVariable=true
   parent=Overlay_33
   widget=Image_GoldenBackground
   properties {brush {
     bIsDynamicallyLoaded=false
     drawAs=Image
     imageType=NoImage
     mirroring=NoMirror
     resourceName=None
     resourceObject=/Game/Athena/HUD/Materials/M_UI_LinearGradient_SpecialPOI
     tiling=NoTile
     imageSize=<256 128>
     margin=<ltrb 0 0 0 0>
     outlineSettings {
      bUseBrushTransparency=false
      roundingType=HalfHeightRadius
      width=0
      color {colorUseRule=UseColor_Specified, specifiedColor=<rgba 0 0 0 0>}
      cornerRadii=<0 0 0 0>
     }
     tintColor {colorUseRule=UseColor_Specified, specifiedColor=<rgba 1 1 1 1>}
     uVRegion {bIsValid=false, max=<0 0>, min=<0 0>}
    }}
   slot {class=OverlaySlot, properties {horizontalAlignment=HAlign_Fill, verticalAlignment=VAlign_Fill}}
  }
  TutorialHighlight {widgetClassPath=&TutorialHighlight}
 ]
}
